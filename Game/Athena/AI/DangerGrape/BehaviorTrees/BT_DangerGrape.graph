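# Behavior tree for the DangerGrape AI character in Athena, driving its gameplay from the aircraft phase onward with leash and target updates.
# A priority selector handles launched and downed states, then diving, gliding and ziplining, then ground behavior.
# Ground behavior puts threat avoidance first, then branches on alert level, with revive, heal and emote options.
# Evaluator services write blackboard statuses that gate each branch, largely reusing Phoebe bot evaluators and subtrees.
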BehaviorTree /Game/Athena/AI/DangerGrape/BehaviorTrees/BT_DangerGrape
blackboard=&/BB_DangerGrape
root {
 name=Gameplay
 class=BTComposite_Selector
 children [
  Wait {
   class=BTTask_Wait
   decorators [Is Character being launched? {
     class=BTDecorator_Blackboard
     properties {
      arithmeticOperation=GreaterOrEqual
      intValue=3
      nodeName=Is Character being launched?
      stringValue=EExecutionStatus::ExecutionPending
      blackboardKey {selectedKeyName=AIEvaluator_CharacterLaunched_ExecutionStatus}
     }
    }]
   properties {waitTime {defaultValue=15}}
  }
  Selector {
   class=BTComposite_Selector
   children [
    Bot Move To {
     class=FortAthenaBTTask_BotMoveTo
     decorators [Can Do DBNO Move {
       class=BTDecorator_Blackboard
       properties {
        arithmeticOperation=GreaterOrEqual
        intValue=3
        nodeName=Can Do DBNO Move
        stringValue=EExecutionStatus::ExecutionPending
        blackboardKey {selectedKeyName=AIEvaluator_DBNO_ExecutionStatus}
       }
      }]
     properties {
      movementResultKeyName=AIEvaluator_DBNO_MovementState
      nodeName=Bot Move To
      acceptableRadius {defaultValue=5}
      allowRandomWobble {defaultValue=true}
      bAllowStrafe {defaultValue=false}
      blackboardKey {selectedKeyName=AIEvaluator_DBNO_Destination}
      observedBlackboardValueTolerance {defaultValue=4.75}
     }
    }
    Wait {class=BTTask_Wait, properties {waitTime {defaultValue=0.5}}}
   ]
   decorators [Is DBNO {
     class=BTDecorator_Blackboard
     properties {arithmeticOperation=GreaterOrEqual, intValue=3, nodeName=Is DBNO, stringValue="", blackboardKey {selectedKeyName=AIEvaluator_Global_DBNO}}
    }]
   properties {
    children [
     {
      childComposite=null
      childTask=&self:FortAthenaBTTask_BotMoveTo_1
      decorators [&self:BTDecorator_Blackboard_5]
     }
     {childComposite=null, childTask=&self:BTTask_Wait_0, decorators []}
    ]
    services [&self:FortAthenaBTService_AIEvaluator_4]
   }
   services [Evaluating...DBNO Move {
     class=FortAthenaBTService_AIEvaluator
     properties {
      bCallTickOnSearchStart=true
      interval=1
      nodeName=Evaluating...DBNO Move
      aIEvaluatorClass {defaultValue=/Game/Athena/AI/Phoebe/Evaluators/Phoebe_Evaluator_DBNO.Phoebe_Evaluator_DBNO_C}
     }
    }]
  }
  Alive {
   class=BTComposite_Selector
   children [
    Diving {
     class=FortAthenaBTTask_Dive
     children []
     decorators [Can Dive {
       class=BTDecorator_Blackboard
       properties {
        arithmeticOperation=Equal
        intValue=4
        nodeName=Can Dive
        stringValue=EExecutionStatus::ExecutionAllowed
        blackboardKey {selectedKeyName=AIEvaluator_Dive_ExecutionStatus}
       }
      }]
     properties {
      bIgnoreRestartSelf=false
      diveDestinationKeyName=AIEvaluator_Dive_Destination
      executionStatusKeyName=AIEvaluator_Dive_ExecutionStatus
      nodeName=Diving
     }
     services []
    }
    Gliding {
     class=FortAthenaBTTask_Glide
     children []
     decorators [Can Glide {
       class=BTDecorator_Blackboard
       properties {
        arithmeticOperation=Equal
        intValue=4
        nodeName=Can Glide
        stringValue=EExecutionStatus::ExecutionAllowed
        blackboardKey {selectedKeyName=AIEvaluator_Glide_ExecutionStatus}
       }
      }]
     properties {
      bIgnoreRestartSelf=false
      executionStatusKeyName=AIEvaluator_Glide_ExecutionStatus
      glideDestinationKeyName=AIEvaluator_Glide_Destination
      nodeName=Gliding
     }
     services []
    }
    Ziplining {
     class=FortAthenaBTTask_Zipline
     children []
     decorators [Is on Zipline {
       class=BTDecorator_Blackboard
       properties {
        arithmeticOperation=GreaterOrEqual
        intValue=4
        nodeName=Is on Zipline
        stringValue=EExecutionStatus::ExecutionAllowed
        blackboardKey {selectedKeyName=AIEvaluator_Zipline_UsageExecutionStatus}
       }
      }]
     properties {
      bIgnoreRestartSelf=false
      nodeName=Ziplining
      usageExecutionStatusName=AIEvaluator_Zipline_UsageExecutionStatus
      ziplineTargetName=AIEvaluator_Zipline_Target
     }
     services []
    }
    On Ground {
     class=BTComposite_Selector
     children [
      AvoidThreat MoveTo {
       class=FortAthenaBTTask_BotMoveTo
       decorators [Can AvoidThreat? {
         class=BTDecorator_Blackboard
         properties {
          arithmeticOperation=GreaterOrEqual
          intValue=4
          nodeName=Can AvoidThreat?
          stringValue=EExecutionStatus::ExecutionAllowed
          blackboardKey {selectedKeyName=AIEvaluator_AvoidThreat_ExecutionStatus}
         }
        }]
       properties {
        movementResultKeyName=AIEvaluator_AvoidThreat_MovementState
        nodeName=AvoidThreat MoveTo
        acceptableRadius {defaultValue=100}
        allowRandomWobble {defaultValue=true}
        bAllowStrafe {defaultValue=true}
        blackboardKey {selectedKeyName=AIEvaluator_AvoidThreat_Destination}
        observedBlackboardValueTolerance {defaultValue=4.75}
       }
      }
      THREATEN {
       class=BTComposite_Selector
       children [Move To Range {
         class=BTTask_RunBehavior
         decorators [
          Can do Range {
           class=BTDecorator_Blackboard
           properties {
            arithmeticOperation=GreaterOrEqual
            intValue=3
            nodeName=Can do Range
            stringValue=EExecutionStatus::ExecutionPending
            blackboardKey {selectedKeyName=AIEvaluator_RangeAttack_ExecutionStatus}
           }
          }
          Loop {class=BTDecorator_Loop}
         ]
         properties {behaviorAsset=&BT_Phoebe_MoveToRange, nodeName=Move To Range}
        }]
       decorators [IsThreatened {
         class=BTDecorator_Blackboard
         properties {
          arithmeticOperation=Equal
          intValue=3
          nodeName=IsThreatened
          stringValue=EAlertLevel::Threatened
          blackboardKey {selectedKeyName=AIEvaluator_Global_AlertLevel}
         }
        }]
       properties {
        nodeName=THREATEN
        children [{
          childComposite=null
          childTask=&self:BTTask_RunBehavior_10
          decorators [
           &self:BTDecorator_Blackboard_35
           &self:BTDecorator_Loop_1
          ]
         }]
        services [
         &self:FortAthenaBTService_AIEvaluator_1
         &self:FortAthenaBTService_AIEvaluator_0
         &self:FortAthenaBTService_ManageWeapon_0
        ]
       }
       services [
        Evaluating...Select Weapon {
         class=FortAthenaBTService_AIEvaluator
         properties {
          interval=0.2
          nodeName=Evaluating...Select Weapon
          randomDeviation=0.05
          aIEvaluatorClass {defaultValue=&FortAthenaAIBotEvaluator_WeaponSelection}
         }
        }
        Evaluating...Range Attack {
         class=FortAthenaBTService_AIEvaluator
         properties {
          bCallTickOnSearchStart=true
          interval=0.2
          nodeName=Evaluating...Range Attack
          randomDeviation=0.05
          aIEvaluatorClass {defaultValue=&FortAthenaAIBotEvaluator_RangeAttack}
         }
        }
        Executing...Weapons Actions {class=FortAthenaBTService_ManageWeapon, properties {interval=1, randomDeviation=0.3}}
       ]
      }
      Not attacking threat {
       class=BTComposite_Selector
       children [
        LKP {
         class=BTComposite_Selector
         children [
          Selector {
           class=BTComposite_Selector
           children [
            Move To LKP {
             class=FortAthenaBTTask_BotMoveTo
             decorators [Should move to LKP {
               class=BTDecorator_Blackboard
               properties {
                arithmeticOperation=Equal
                intValue=0
                nodeName=Should move to LKP
                stringValue=""
                blackboardKey {selectedKeyName=AIEvaluator_Investigate_MoveToDestination}
               }
              }]
             properties {
              movementResultKeyName=AIEvaluator_Investigate_MovementState
              nodeName=Move To LKP
              acceptableRadius {defaultValue=50}
              allowRandomWobble {defaultValue=true}
              bAllowStrafe {defaultValue=false}
              bReachTestIncludesAgentRadius {defaultValue=false}
              bReachTestIncludesGoalRadius {defaultValue=false}
              bTrackMovingGoal {defaultValue=false}
              blackboardKey {selectedKeyName=AIEvaluator_Investigate_Destination}
              observedBlackboardValueTolerance {defaultValue=4.75}
             }
            }
            Wait {class=BTTask_Wait, properties {randomDeviation {defaultValue=0.5}, waitTime {defaultValue=0.75}}}
           ]
           decorators [Should Investigate {
             class=BTDecorator_Blackboard
             properties {
              arithmeticOperation=Equal
              intValue=4
              nodeName=Should Investigate
              stringValue=EExecutionStatus::ExecutionAllowed
              blackboardKey {selectedKeyName=AIEvaluator_Investigate_ExecutionStatus}
             }
            }]
           properties {
            children [
             {
              childComposite=null
              childTask=&self:FortAthenaBTTask_BotMoveTo_3
              decorators [&self:BTDecorator_Blackboard_8]
             }
             {childComposite=null, childTask=&self:BTTask_Wait_12, decorators []}
            ]
            services []
           }
           services []
          }
          Fallback Behavior {
           class=BTComposite_Selector
           children [
            Run PatrolAround Behavior {
             class=BTTask_RunBehavior
             decorators [Patrol Around {
               class=BTDecorator_Blackboard
               properties {
                arithmeticOperation=GreaterOrEqual
                intValue=2
                nodeName=Patrol Around
                stringValue=EExecutionStatus::ExecutionSuccess
                blackboardKey {selectedKeyName=AIEvaluator_Patrol_ExecutionStatus}
               }
              }]
             properties {behaviorAsset=/Game/Athena/AI/Phoebe/BehaviorTrees/BT_Phoebe_PatrolAround, nodeName=Run PatrolAround Behavior}
            }
            Wait {class=BTTask_Wait, properties {waitTime {defaultValue=2}}}
           ]
           decorators []
           properties {
            nodeName=Fallback Behavior
            children [
             {
              childComposite=null
              childTask=&self:BTTask_RunBehavior_14
              decorators [&self:BTDecorator_Blackboard_28]
             }
             {childComposite=null, childTask=&self:BTTask_Wait_3, decorators []}
            ]
            services [&self:FortAthenaBTService_AIEvaluator_12]
           }
           services [Evaluating ... Patrol Around {
             class=FortAthenaBTService_AIEvaluator
             properties {
              bCallTickOnSearchStart=true
              interval=2.5
              nodeName=Evaluating ... Patrol Around
              randomDeviation=0.5
              aIEvaluatorClass {defaultValue=/Script/FortniteAIServer.FortAthenaAIBotEvaluator_PatrolAround}
             }
            }]
          }
         ]
         decorators [IsLKP {
           class=BTDecorator_Blackboard
           properties {
            arithmeticOperation=Equal
            intValue=2
            nodeName=IsLKP
            stringValue=EAlertLevel::LKP
            blackboardKey {selectedKeyName=AIEvaluator_Global_AlertLevel}
           }
          }]
         properties {
          nodeName=LKP
          children [
           {
            childComposite=&self:BTComposite_Selector_12
            childTask=null
            decorators [&self:BTDecorator_Blackboard_27]
           }
           {childComposite=&self:BTComposite_Selector_8, childTask=null, decorators []}
          ]
          services [
           &self:FortAthenaBTService_AIEvaluator_10
           &self:FortAthenaBTService_AIEvaluator_20
           &self:FortAthenaBTService_AIEvaluator_32
           &self:FortAthenaBTService_AIEvaluator_33
           &self:FortAthenaBTService_ManageWeapon_3
          ]
         }
         services [
          Evaluating...Select Weapon {
           class=FortAthenaBTService_AIEvaluator
           properties {
            interval=0.2
            nodeName=Evaluating...Select Weapon
            randomDeviation=0.05
            aIEvaluatorClass {defaultValue=&FortAthenaAIBotEvaluator_WeaponSelection}
           }
          }
          Evaluating...Investigate Treat {
           class=FortAthenaBTService_AIEvaluator
           properties {
            bCallTickOnSearchStart=true
            interval=0.5
            nodeName=Evaluating...Investigate Treat
            aIEvaluatorClass {defaultValue=/Script/FortniteAIServer.FortAthenaAIBotEvaluator_Investigate}
           }
          }
          Evaluating...Range Attack {
           class=FortAthenaBTService_AIEvaluator
           properties {
            interval=0.3
            nodeName=Evaluating...Range Attack
            aIEvaluatorClass {defaultValue=&FortAthenaAIBotEvaluator_RangeAttack}
           }
          }
          Evaluating...Melee Attack {
           class=FortAthenaBTService_AIEvaluator
           properties {
            interval=0.3
            nodeName=Evaluating...Melee Attack
            aIEvaluatorClass {defaultValue=&FortAthenaAIBotEvaluator_MeleeAttack}
           }
          }
          Executing...Weapons Actions {class=FortAthenaBTService_ManageWeapon, properties {interval=0.2, randomDeviation=0.1}}
         ]
        }
        ALERTED {
         class=BTComposite_Selector
         children [
          LookAtThreat {
           class=BTComposite_Selector
           children [
            MoveDirectlyToward {
             class=BTTask_MoveDirectlyToward
             children []
             decorators [Can Dodge {
               class=BTDecorator_Blackboard
               properties {
                arithmeticOperation=GreaterOrEqual
                intValue=2
                nodeName=Can Dodge
                stringValue=EExecutionStatus::ExecutionSuccess
                blackboardKey {selectedKeyName=AIEvaluator_Dodge_ExecutionStatus}
               }
              }]
             properties {
              bIgnoreRestartSelf=false
              nodeName=MoveDirectlyToward
              acceptableRadius {defaultValue=5, key=None}
              bAllowPartialPath {defaultValue=true, key=None}
              bAllowStrafe {defaultValue=true, key=None}
              bProjectGoalLocation {defaultValue=true, key=None}
              bReachTestIncludesAgentRadius {defaultValue=false, key=None}
              bReachTestIncludesGoalRadius {defaultValue=false, key=None}
              bRequireNavigableEndLocation {defaultValue=true, key=None}
              bStartFromPreviousPath {defaultValue=false, key=None}
              bTrackMovingGoal {defaultValue=true, key=None}
              blackboardKey {bNoneIsAllowedValue=false, selectedKeyName=AIEvaluator_Dodge_Destination}
              filterClass {baseClass=&NavigationQueryFilter, defaultValue=null, key=None}
              observedBlackboardValueTolerance {defaultValue=4.75, key=None}
             }
             services []
            }
            Wait {class=BTTask_Wait, properties {randomDeviation {defaultValue=0.25}, waitTime {defaultValue=1}}}
           ]
           decorators [Should Observe {
             class=BTDecorator_Blackboard
             properties {
              arithmeticOperation=GreaterOrEqual
              intValue=3
              nodeName=Should Observe
              stringValue=EExecutionStatus::ExecutionPending
              blackboardKey {selectedKeyName=AIEvaluator_Observe_ExecutionStatus}
             }
            }]
           properties {
            bApplyDecoratorScope=true
            nodeName=LookAtThreat
            children [
             {
              childComposite=null
              childTask=&self:BTTask_MoveDirectlyToward_1
              decorators [&self:BTDecorator_Blackboard_33]
             }
             {childComposite=null, childTask=&self:BTTask_Wait_2, decorators []}
            ]
            services [
             &self:FortAthenaBTService_AIEvaluator_29
             &self:FortAthenaBTService_AIEvaluator_40
            ]
           }
           services [
            Focus on target {
             class=FortAthenaBTService_AIEvaluator
             properties {interval=0.5, nodeName=Focus on target, aIEvaluatorClass {defaultValue=&FortAthenaAIBotEvaluator_HandleFocusing}}
            }
            Evaluating... Evasive Maneuvers {
             class=FortAthenaBTService_AIEvaluator
             properties {
              interval=0.5
              nodeName=Evaluating... Evasive Maneuvers
              aIEvaluatorClass {defaultValue=/Game/Athena/AI/Phoebe/Evaluators/Phoebe_EvasiveManeuvers.Phoebe_EvasiveManeuvers_C}
             }
            }
           ]
          }
          Fallback Behavior {
           class=BTComposite_Selector
           children [
            Run PatrolAround Behavior {
             class=BTTask_RunBehavior
             decorators [Patrol Around {
               class=BTDecorator_Blackboard
               properties {
                arithmeticOperation=GreaterOrEqual
                intValue=2
                nodeName=Patrol Around
                stringValue=EExecutionStatus::ExecutionSuccess
                blackboardKey {selectedKeyName=AIEvaluator_Patrol_ExecutionStatus}
               }
              }]
             properties {behaviorAsset=&BT_Phoebe_PatrolAround, nodeName=Run PatrolAround Behavior}
            }
            Wait {class=BTTask_Wait, properties {waitTime {defaultValue=2}}}
           ]
           decorators []
           properties {
            nodeName=Fallback Behavior
            children [
             {
              childComposite=null
              childTask=&self:BTTask_RunBehavior_15
              decorators [&self:BTDecorator_Blackboard_24]
             }
             {childComposite=null, childTask=&self:BTTask_Wait_4, decorators []}
            ]
            services [&self:FortAthenaBTService_AIEvaluator_46]
           }
           services [Evaluating ... Patrol Around {
             class=FortAthenaBTService_AIEvaluator
             properties {
              bCallTickOnSearchStart=true
              interval=2.5
              nodeName=Evaluating ... Patrol Around
              randomDeviation=0.5
              aIEvaluatorClass {defaultValue=&FortAthenaAIBotEvaluator_PatrolAround}
             }
            }]
          }
         ]
         decorators [
          IsAlerted {
           class=BTDecorator_Blackboard
           properties {
            arithmeticOperation=Equal
            intValue=1
            nodeName=IsAlerted
            stringValue=EAlertLevel::Alerted
            blackboardKey {selectedKeyName=AIEvaluator_Global_AlertLevel}
           }
          }
          Loop {class=BTDecorator_Loop, properties {infiniteLoopTimeoutTime {defaultValue=4}}}
         ]
         properties {
          nodeName=ALERTED
          children [
           {
            childComposite=&self:BTComposite_Selector_11
            childTask=null
            decorators [&self:BTDecorator_Blackboard_13]
           }
           {childComposite=&self:BTComposite_Selector_14, childTask=null, decorators []}
          ]
          services [
           &self:FortAthenaBTService_AIEvaluator_31
           &self:FortAthenaBTService_AIEvaluator_55
           &self:FortAthenaBTService_AIEvaluator_61
           &self:FortAthenaBTService_ManageWeapon_8
          ]
         }
         services [
          Evaluating...Select Weapon {
           class=FortAthenaBTService_AIEvaluator
           properties {
            interval=0.2
            nodeName=Evaluating...Select Weapon
            randomDeviation=0.05
            aIEvaluatorClass {defaultValue=&FortAthenaAIBotEvaluator_WeaponSelection}
           }
          }
          Evaluating...Observe Treat {
           class=FortAthenaBTService_AIEvaluator
           properties {
            bCallTickOnSearchStart=true
            interval=0.5
            nodeName=Evaluating...Observe Treat
            aIEvaluatorClass {defaultValue=/Script/FortniteAIServer.FortAthenaAIBotEvaluator_Observe}
           }
          }
          Evaluating...Melee Attack {
           class=FortAthenaBTService_AIEvaluator
           properties {
            interval=0.3
            nodeName=Evaluating...Melee Attack
            aIEvaluatorClass {defaultValue=&FortAthenaAIBotEvaluator_MeleeAttack}
           }
          }
          Executing...Weapons Actions {class=FortAthenaBTService_ManageWeapon, properties {interval=0.2, randomDeviation=0.1}}
         ]
        }
        Run Behavior {
         class=BTTask_RunBehavior
         decorators [Revive {
           class=BTDecorator_Blackboard
           properties {
            arithmeticOperation=GreaterOrEqual
            intValue=3
            nodeName=Revive
            stringValue=EExecutionStatus::ExecutionPending
            blackboardKey {selectedKeyName=AIEvaluator_Revive_ExecutionStatus}
           }
          }]
         properties {behaviorAsset=/Game/Athena/AI/Phoebe/BehaviorTrees/BT_Phoebe_Revive, nodeName=Run Behavior}
        }
        Use Heal Item {
         class=FortAthenaBTTask_UseItem
         decorators [Can Heal {
           class=BTDecorator_Blackboard
           properties {
            arithmeticOperation=GreaterOrEqual
            intValue=3
            nodeName=Can Heal
            stringValue=EExecutionStatus::ExecutionPending
            blackboardKey {selectedKeyName=AIEvaluator_Healing_ExecutionStatus}
           }
          }]
         properties {
          actionObjectKeyName=AIEvaluator_Healing_Object
          executionStatusKeyName=AIEvaluator_Healing_ExecutionStatus
          nodeName=Use Heal Item
          finishAfterItemUse {defaultValue=false}
          inputHoldDuration {key=None}
          maximumWaitTimeBetweenUses {defaultValue=1}
          resetActionObjectKey {defaultValue=true}
         }
        }
        UNAWARE {
         class=BTComposite_Selector
         children [Bot Move To {
           class=FortAthenaBTTask_BotMoveTo
           decorators []
           properties {
            movementResultKeyName=None
            nodeName=Bot Move To
            acceptableRadius {defaultValue=5}
            allowRandomWobble {defaultValue=true}
            bAllowStrafe {defaultValue=false}
            blackboardKey {selectedKeyName=AIEvaluator_NotAlerted_MoveToDestination}
            observedBlackboardValueTolerance {defaultValue=4.75}
           }
          }]
         decorators [Loop {class=BTDecorator_Loop}]
         properties {
          nodeName=UNAWARE
          children [{childComposite=null, childTask=&self:FortAthenaBTTask_BotMoveTo_4, decorators []}]
          services [
           &self:FortAthenaBTService_AIEvaluator_64
           &self:FortAthenaBTService_AIEvaluator_39
           &self:FortAthenaBTService_AIEvaluator_53
           &self:FortAthenaBTService_ManageWeapon_2
           &self:FortAthenaBTService_AIEvaluator_54
          ]
         }
         services [
          Evaluating...Select Weapon {
           class=FortAthenaBTService_AIEvaluator
           properties {
            interval=0.2
            nodeName=Evaluating...Select Weapon
            randomDeviation=0.05
            aIEvaluatorClass {defaultValue=&FortAthenaAIBotEvaluator_WeaponSelection}
           }
          }
          Evaluating...Range Attack {
           class=FortAthenaBTService_AIEvaluator
           properties {
            interval=0.3
            nodeName=Evaluating...Range Attack
            aIEvaluatorClass {defaultValue=&FortAthenaAIBotEvaluator_RangeAttack}
           }
          }
          Evaluating...Melee Attack {
           class=FortAthenaBTService_AIEvaluator
           properties {
            interval=0.3
            nodeName=Evaluating...Melee Attack
            aIEvaluatorClass {defaultValue=&FortAthenaAIBotEvaluator_MeleeAttack}
           }
          }
          Executing...Weapons Actions {class=FortAthenaBTService_ManageWeapon, properties {interval=0.2, randomDeviation=0.1}}
          Focus Scan Around {
           class=FortAthenaBTService_AIEvaluator
           properties {
            interval=0.5
            nodeName=Focus Scan Around
            aIEvaluatorClass {defaultValue=/Game/Athena/AI/Phoebe/Evaluators/Phoebe_HandleFocusing_ScanAround.Phoebe_HandleFocusing_ScanAround_C}
           }
          }
         ]
        }
        PlayEmote {
         class=FortAthenaBTTask_PlayEmote
         children []
         decorators [Can DanceOnKill? {
           class=BTDecorator_Blackboard
           properties {
            arithmeticOperation=GreaterOrEqual
            flowAbortMode=LowerPriority
            intValue=3
            nodeName=Can DanceOnKill?
            stringValue=EExecutionStatus::ExecutionPending
            blackboardKey {selectedKeyName=AIEvaluator_PlayEmote_ExecutionStatus}
           }
          }]
         properties {bIgnoreRestartSelf=false, nodeName=PlayEmote, playEmoteExecutionStatusKeyName=AIEvaluator_PlayEmote_ExecutionStatus, bespokeEmotes []}
         services []
        }
       ]
       decorators []
       properties {
        nodeName=Not attacking threat
        children [
         {
          childComposite=&self:BTComposite_Selector_6
          childTask=null
          decorators [&self:BTDecorator_Blackboard_20]
         }
         {
          childComposite=&self:BTComposite_Selector_7
          childTask=null
          decorators [
           &self:BTDecorator_Blackboard_17
           &self:BTDecorator_Loop_4
          ]
         }
         {
          childComposite=null
          childTask=&self:BTTask_RunBehavior_7
          decorators [&self:BTDecorator_Blackboard_29]
         }
         {
          childComposite=null
          childTask=&self:FortAthenaBTTask_UseItem_1
          decorators [&self:BTDecorator_Blackboard_10]
         }
         {
          childComposite=&self:BTComposite_Selector_2
          childTask=null
          decorators [&self:BTDecorator_Loop_2]
         }
         {
          childComposite=null
          childTask=&self:FortAthenaBTTask_PlayEmote_0
          decorators [&self:BTDecorator_Blackboard_19]
         }
        ]
        services [
         &self:FortAthenaBTService_AIEvaluator_17
         &self:FortAthenaBTService_AIEvaluator_36
        ]
       }
       services [
        Evaluating...Healing {
         class=FortAthenaBTService_AIEvaluator
         properties {
          interval=0.5
          nodeName=Evaluating...Healing
          aIEvaluatorClass {defaultValue=/Game/Athena/AI/Phoebe/Evaluators/Phoebe_Evaluator_Healing.Phoebe_Evaluator_Healing_C}
         }
        }
        Evaluating...DanceOnKill {
         class=FortAthenaBTService_AIEvaluator
         properties {
          bCallTickOnSearchStart=true
          interval=0.5
          nodeName=Evaluating...DanceOnKill
          aIEvaluatorClass {defaultValue=/Script/FortniteAIServer.FortAthenaAIBotEvaluator_DanceOnKill}
         }
        }
       ]
      }
     ]
     decorators [BTDecorator_HasEverSkydivedFromBusAndLanded_C_0 {
       class=BTDecorator_HasEverSkydivedFromBusAndLanded_C
       properties {
        bCheckConditionOnlyBlackBoardChanges=false
        bInverseCondition=false
        bShowPropertyDetails=true
        customDescription=""
        flowAbortMode=None
        nodeName=""
       }
      }]
     properties {
      nodeName=On Ground
      children [
       {
        childComposite=null
        childTask=&self:FortAthenaBTTask_BotMoveTo_0
        decorators [&self:BTDecorator_Blackboard_40]
       }
       {
        childComposite=&self:BTComposite_Selector_4
        childTask=null
        decorators [&self:BTDecorator_Blackboard_4]
       }
       {childComposite=&self:BTComposite_Selector_15, childTask=null, decorators []}
      ]
      services [
       &self:FortAthenaBTService_AIEvaluator_14
       &self:FortAthenaBTService_AIEvaluator_11
       &self:FortAthenaBTService_AIEvaluator_43
       &self:FortAthenaBTService_AIEvaluator_47
       &self:FortAthenaBTService_Crouch_1
       &self:FortAthenaBTService_Jump_1
       &self:FortAthenaBTService_Sprinting_1
       &self:FortAthenaBTService_BuildConstruction_0
       &self:FortAthenaBTService_AIEvaluator_49
      ]
     }
     services [
      Evaluate...Sprinting {
       class=FortAthenaBTService_AIEvaluator
       properties {interval=0.2, nodeName=Evaluate...Sprinting, aIEvaluatorClass {defaultValue=/Script/FortniteAIServer.FortAthenaAIBotEvaluator_Sprinting}}
      }
      Evaluate...Revive {
       class=FortAthenaBTService_AIEvaluator
       properties {interval=0.5, nodeName=Evaluate...Revive, aIEvaluatorClass {defaultValue=/Script/FortniteAIServer.FortAthenaAIBotEvaluator_Revive}}
      }
      Evaluate...AvoidThreat {
       class=FortAthenaBTService_AIEvaluator
       properties {
        interval=0.5
        nodeName=Evaluate...AvoidThreat
        aIEvaluatorClass {defaultValue=/Game/Athena/AI/Phoebe/Evaluators/Phoebe_Evaluator_AvoidThreat.Phoebe_Evaluator_AvoidThreat_C}
       }
      }
      Evaluate...StealWall {
       class=FortAthenaBTService_AIEvaluator
       properties {interval=10, nodeName=Evaluate...StealWall, aIEvaluatorClass {defaultValue=/Script/FortniteAIServer.FortAthenaAIBotEvaluator_StealWall}}
      }
      Executing...Crouch {class=FortAthenaBTService_Crouch}
      Executing...Jump {class=FortAthenaBTService_Jump, properties {crouchExecutionStatusName=AIEvaluator_Crouch_ExecutionStatus}}
      Executing...Sprinting {class=FortAthenaBTService_Sprinting}
      Executing...Build Construction {
       class=FortAthenaBTService_BuildConstruction
       properties {
        bCallTickOnSearchStart=false
        bRestartTimerOnEachActivation=false
        interval=0.5
        nodeName=Executing...Build Construction
        randomDeviation=0.1
        stealWallBuildGridCoordName=AIEvaluator_StealWallBuilding_GridCoord
        stealWallBuildName=AIEvaluator_StealWallBuilding_ExecutionStatus
        stealWallBuildTypeName=AIEvaluator_StealWallBuilding_Type
       }
      }
      Evaluating...Danger {
       class=FortAthenaBTService_AIEvaluator
       properties {
        interval=1000000000
        nodeName=Evaluating...Danger
        randomDeviation=0
        aIEvaluatorClass {defaultValue=/Game/Athena/AI/Phoebe/Evaluators/Phoebe_Evaluator_DangerDetection.Phoebe_Evaluator_DangerDetection_C}
       }
      }
     ]
    }
   ]
   decorators []
   properties {
    nodeName=Alive
    children [
     {
      childComposite=null
      childTask=&self:FortAthenaBTTask_Dive_0
      decorators [&self:BTDecorator_Blackboard_12]
     }
     {
      childComposite=null
      childTask=&self:FortAthenaBTTask_Glide_0
      decorators [&self:BTDecorator_Blackboard_18]
     }
     {
      childComposite=null
      childTask=&self:FortAthenaBTTask_Zipline_0
      decorators [&self:BTDecorator_Blackboard_50]
     }
     {
      childComposite=&self:BTComposite_Selector_20
      childTask=null
      decorators [&self:BTDecorator_HasEverSkydivedFromBusAndLanded_C_0]
     }
    ]
    services [&self:FortAthenaBTService_AIEvaluator_9]
   }
   services [Evaluating...Free Fall {
     class=FortAthenaBTService_AIEvaluator
     properties {
      bCallTickOnSearchStart=true
      interval=0.5
      nodeName=Evaluating...Free Fall
      aIEvaluatorClass {defaultValue=/Script/FortniteAIServer.FortAthenaAIBotEvaluator_FreeFalling}
     }
    }]
  }
 ]
 decorators [In Gameplay? {
   class=BTDecorator_Blackboard
   properties {
    arithmeticOperation=GreaterOrEqual
    flowAbortMode=None
    intValue=3
    nodeName=In Gameplay?
    stringValue=EAthenaGamePhase::Aircraft
    blackboardKey {selectedKeyName=AIEvaluator_Global_GamePhase}
   }
  }]
 properties {
  nodeName=Gameplay
  children [
   {
    childComposite=null
    childTask=&self:BTTask_Wait_6
    decorators [&self:BTDecorator_Blackboard_14]
   }
   {
    childComposite=&self:BTComposite_Selector_5
    childTask=null
    decorators [&self:BTDecorator_Blackboard_0]
   }
   {childComposite=&self:BTComposite_Selector_10, childTask=null, decorators []}
  ]
  services [
   &self:FortAthenaBTService_AIEvaluator_3
   &self:FortAthenaBTService_UpdateTarget_0
   &self:FortAthenaBTService_AIEvaluator_2
  ]
 }
 services [
  Evaluating... Leash {
   class=FortAthenaBTService_AIEvaluator
   properties {interval=0.5, nodeName=Evaluating... Leash, aIEvaluatorClass {defaultValue=/Script/FortniteAIServer.FortAthenaNpcEvaluator_Leash}}
  }
  Evaluating...Targets {class=FortAthenaBTService_UpdateTarget, properties {interval=0.1}}
  Evaluating...Character Launched {
   class=FortAthenaBTService_AIEvaluator
   properties {
    interval=0.5
    nodeName=Evaluating...Character Launched
    aIEvaluatorClass {defaultValue=/Script/FortniteAIServer.FortAthenaAIBotEvaluator_CharacterLaunched}
   }
  }
 ]
}
